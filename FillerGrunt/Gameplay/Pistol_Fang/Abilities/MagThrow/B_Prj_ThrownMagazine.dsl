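; Thrown magazine projectile for the Fang pistol's MagThrow ability, which spins in flight and self-destructs when its lifetime timer expires.
; Explodes asynchronously on authority hits or non-overlap touches; OnTouched also clears the timer and plays client impact sound, covering hits that precede client spawn.
; Homing turn rates are set upon nearing or passing the target.
; OnStop fades the in-air audio loop, since OnTouched can be missed on clients.
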(event EventTick
  (params (DeltaSeconds :type Float%20%28single-precision%29))
  (Transformation|AddLocalRotation (Variables|Default|GetPivotPointMag) (Math|Rotator|MakeRotator 0.0 (* DeltaSeconds -4.5e+03)) false false))

(event EventBeginPlay
  ;[0.02391 0.15 0.02888 0.5] Begin Play
  (bind _returnvalue (Utilities|Time|SetTimerbyEvent (EventDispatchers|CreateEvent :!event Custom|LifetimeEnded) 5.0))
  (Variables|Default|SetLifetimeTimer _returnvalue))

(event Collision|EventHit
  (params (MyComp :type Primitive%20Component%20Object%20Reference) (Other :type Actor%20Object%20Reference) (OtherComp :type Primitive%20Component%20Object%20Reference) (bSelfMoved :type Boolean) (HitLocation :type Vector :default "0, 0, 0") (HitNormal :type Vector :default "0, 0, 0") (NormalImpulse :type Vector :default "0, 0, 0") (Hit :type Hit%20Result%20Structure%20%28by%20ref%29))
  (|CallParentFunction MyComp Other OtherComp bSelfMoved HitLocation HitNormal NormalImpulse Hit)
  (Networking|SwitchHasAuthority
    (:Authority
      (Utilities|FlowControl|DoOnce false
        (:Completed
          (Delivery|DoExplosionAsync))))
    (:Remote
      (Development|PrintString "Client Impact"))))

(event Custom|LifetimeEnded
  ;[0.15 0.04442 0.02717 0.5] Lifetime Ended
  (Game|Components|ProjectileMovement|StopSimulating (Variables|Movement|GetProjectileMovementComponent) (Collision|MakeHitResult false false 0.0 0.0 (Transformation|GetActorLocation) "0, 0, 0" (Math|Vector|Normalize (Transformation|GetVelocity))))
  (Actor|DestroyActor))

(event Custom|PlayHitSound
  (Development|PrintString "HIT SOUND")
  (Audio|PlaySoundatLocation (Variables|Default|GetImpact_SFX) (Transformation|GetWorldLocation (Variables|Default|GetMesh_Magazine)) "0, 0, 0" 1.0 1.0 0.0 0 0 (Actor|GetOwner)))

(event EventOnTouched
  (params (OtherActor :type Actor%20Object%20Reference) (OtherComp :type Primitive%20Component%20Object%20Reference) (HitResult :type Hit%20Result%20Structure%20%28by%20ref%29) (bIsOverlap :type Boolean))
  ;[0.04325 0.15 0.09947 0.5] On Touched
  ; Used instead of OnHit because it is called on client by parent class if hit occurs before client projectile spawned
  (|CallParentFunction OtherActor OtherComp HitResult bIsOverlap)
  (if (not bIsOverlap)
    (Delivery|DoExplosionAsync)
    (Utilities|Time|ClearandInvalidateTimerbyHandle (Variables|Default|GetLifetimeTimer))
    (if (not (Networking|IsServer))
      (Development|PrintString "I AM A CLIENT HIT")
      (CallFunction|PlayHitSound))))

(event Custom|OnPickup_Event
  (params (SelfActor :type Fort%20Pickup%20Object%20Reference) (InteractingPawn :type Fort%20Pawn%20Object%20Reference))
  (Development|PrintString "Hello"))

(event ProjectileMovement|EventOnReachedLockTargetDistanceThreshold
  (Projectile|SetHomingTurnSpeedMin 0 0 3.6e+02 :!class /Script/FortniteGame.FortProjectileMovementInterface)
  (Projectile|SetHomingTurnSpeedMax 0 0 7.2e+02 :!class /Script/FortniteGame.FortProjectileMovementInterface)
  (Projectile|SetHomingTimeUntilMaxTurnSpeed 0 0 0.5 :!class /Script/FortniteGame.FortProjectileMovementInterface))

(event ProjectileMovement|EventOnHomingPassedTarget
  (Projectile|SetHomingTurnSpeedMin 0 0 3.6e+02 :!class /Script/FortniteGame.FortProjectileMovementInterface)
  (Projectile|SetHomingTurnSpeedMax 0 0 7.2e+02 :!class /Script/FortniteGame.FortProjectileMovementInterface)
  (Projectile|SetHomingTimeUntilMaxTurnSpeed 0 0 0.5 :!class /Script/FortniteGame.FortProjectileMovementInterface))

(event EventOnStop
  (params (Hit :type Hit%20Result%20Structure%20%28by%20ref%29))
  ;[0.15 0.06903 0.02658 0.5] On Stop
  ; Used to ensure any time projectile is stopped, in air FX stop playing. Sometimes OnTouched is not called on client
  (|CallParentFunction Hit)
  (Networking|CanExecuteCosmeticEvents
    (:True
      (Audio|Components|Audio|FadeOut (Variables|Default|GetAudio_InAirLoop) 0.1 0.0 "Linear" :!class /Script/Engine.AudioComponent))))

(fn ConstructionScript
  (|CallParentFunction))
