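; Reload condition for the FillerGrunt Fang pistol that reports whether the mag-throw targeting has a valid enemy in range.
; The console variable FillerGrunt.MagThrowTargeting.EnableNative selects between two evaluation paths.
; The native path checks the owning player pawn's ability system for the MagThrow InRange gameplay tag.
; The Blueprint path reads MagThrow_ValidInRange from the pistol actor; failed casts or validity checks return false.
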(fn EvaluateCondition
  (params (ReloadStateData :type Segmented%20Reload%20State%20Data%20Structure%20%28by%20ref%29) (WeaponActor :type Fort%20Weapon%20Object%20Reference))
  (bind _returnvalue (Development|GetConsoleVariableBoolValue "FillerGrunt.MagThrowTargeting.EnableNative"))
  (if _returnvalue
    ; Nativized Version
    (Utilities|IsValid (Variables|GetWeaponActor)
      (:Is%20Valid
        (bind _asfort_player_pawn (Utilities|Casting|CastToFortPlayerPawn (Actor|GetOwner (Variables|GetWeaponActor)))
          (:then
            (Utilities|IsValid (Class|FortAbilitySystemInterface|GetAbilitySystemComponentDontCreate _asfort_player_pawn)
              (:Is%20Valid
                (return (GameplayTags|HasMatchingGameplayTag (Class|FortAbilitySystemInterface|GetAbilitySystemComponentDontCreate _asfort_player_pawn) "(TagName=\"Ability.FillerGrunt.MagThrow.Targeting.InRange\")")))
              (:Is%20Not%20Valid
                (return false))))
          (:CastFailed
            (return false))))
      (:Is%20Not%20Valid
        (return false)))
    (else
      ; BP Version
      (bind _asbp_fg_fang_pistol (Utilities|Casting|CastToBP_FG_Fang_Pistol (Variables|GetWeaponActor))
        (:then
          (return (Variables|/FillerGrunt/Gameplay/Pistol_Fang/BP_FG_Fang_Pistol.BP_FG_Fang_Pistol_C|GetMagThrow_ValidInRange _asbp_fg_fang_pistol)))
        (:CastFailed
          (return false))))))
