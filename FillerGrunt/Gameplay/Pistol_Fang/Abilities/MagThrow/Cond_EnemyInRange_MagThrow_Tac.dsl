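; Condition for the Fang pistol's MagThrow that passes when an enemy is targetable and a reload-state value is above zero.
; A console variable selects a native gameplay-tag check or a Blueprint trace using curve-table values, which caches its result on the pistol.
; The trace box-sweeps for pawns, then applies target filtering, a pyramid-cone test and high/low stance-based line-of-sight checks.
; Vehicle occupants skip line-of-sight; controllerless owners skip tracing and pass range.
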(fn CacheStatData
  (bind (_outxy _returnvalue) (Fortnite|EvaluateCurveTableRow (Variables|Spawn|GetSpawnCrouchedOffsetData)))
  (Variables|Spawn|SetCrouchedOffset_Cached (Math|Vector|MakeVector 0.0 0.0 (Math|Float|SelectFloat _outxy 35.0 _returnvalue)))
  (bind (_outxy_2 _returnvalue_2) (Fortnite|EvaluateCurveTableRow (Variables|Spawn|GetSpawnStandingOffsetData)))
  (Variables|Spawn|SetStandingOffset_Cached (Math|Vector|MakeVector 0.0 0.0 (Math|Float|SelectFloat _outxy_2 7e+01 _returnvalue_2)))
  (bind (_outxy_3 _returnvalue_3) (Fortnite|EvaluateCurveTableRow (Variables|Spawn|GetSpawnVehicleOffsetData)))
  (Variables|Spawn|SetSpawnVehicleOffset_Cached (Math|Vector|MakeVector 0.0 0.0 (Math|Float|SelectFloat _outxy_3 35.0 _returnvalue_3)))
  (bind (_outxy_4 _returnvalue_4) (Fortnite|EvaluateCurveTableRow (Variables|Default|GetRightOffsetData)))
  (Variables|Default|SetRightOffset_Cached (Math|Float|SelectFloat _outxy_4 2e+01 _returnvalue_4))
  (bind (_outxy_5 _returnvalue_5) (Fortnite|EvaluateCurveTableRow (Variables|Default|GetTargetingRangeData)))
  (Variables|Default|SetTargetingRange_Cached (Math|Float|SelectFloat _outxy_5 2e+03 _returnvalue_5))
  (bind (_outxy_6 _returnvalue_6) (Fortnite|EvaluateCurveTableRow (Variables|Default|GetTargetingConePitchData)))
  (Variables|Default|SetTargetingConePitch_Cached (Math|Float|SelectFloat _outxy_6 45.0 _returnvalue_6))
  (bind (_outxy_7 _returnvalue_7) (Fortnite|EvaluateCurveTableRow (Variables|Default|GetTargetingConeYawData)))
  (Variables|Default|SetTargetingConeYaw_Cached (Math|Float|SelectFloat _outxy_7 8e+01 _returnvalue_7))
  (bind (_outxy_8 _returnvalue_8) (Fortnite|EvaluateCurveTableRow (Variables|LOSCheck|GetLOS_Check_Vehicle_High_Data)))
  (Variables|LOSCheck|SetLOSCheck_Vehicle_High_Cached (Math|Vector|MakeVector 0.0 0.0 (Math|Float|SelectFloat _outxy_8 4e+01 _returnvalue_8)))
  (bind (_outxy_9 _returnvalue_9) (Fortnite|EvaluateCurveTableRow (Variables|LOSCheck|GetLOS_Check_Vehicle_Low_Data)))
  (Variables|LOSCheck|SetLOSCheck_Vehicle_Low_Cached (Math|Vector|MakeVector 0.0 0.0 (Math|Float|SelectFloat _outxy_9 0.0 _returnvalue_9)))
  (bind (_outxy_10 _returnvalue_10) (Fortnite|EvaluateCurveTableRow (Variables|LOSCheck|GetLOS_Check_Standing_High_Data)))
  (Variables|LOSCheck|SetLOSCheck_Standing_High_Cached (Math|Vector|MakeVector 0.0 0.0 (Math|Float|SelectFloat _outxy_10 7e+01 _returnvalue_10)))
  (bind (_outxy_11 _returnvalue_11) (Fortnite|EvaluateCurveTableRow (Variables|LOSCheck|GetLOS_Check_Standing_Low_Data)))
  (Variables|LOSCheck|SetLOSCheck_Standing_Low_Cached (Math|Vector|MakeVector 0.0 0.0 (Math|Float|SelectFloat _outxy_11 35.0 _returnvalue_11)))
  (bind (_outxy_12 _returnvalue_12) (Fortnite|EvaluateCurveTableRow (Variables|LOSCheck|GetLOS_Check_Crouched_High_Data)))
  (Variables|LOSCheck|SetLOSCheck_Crouching_High_Cached (Math|Vector|MakeVector 0.0 0.0 (Math|Float|SelectFloat _outxy_12 35.0 _returnvalue_12)))
  (bind (_outxy_13 _returnvalue_13) (Fortnite|EvaluateCurveTableRow (Variables|LOSCheck|GetLOS_Check_Crouched_Low_Data)))
  (Variables|LOSCheck|SetLOSCheck_Crouching_Low_Cached (Math|Vector|MakeVector 0.0 0.0 (Math|Float|SelectFloat _outxy_13 17.5 _returnvalue_13)))
  (bind (_outxy_14 _returnvalue_14) (Fortnite|EvaluateCurveTableRow (Variables|LOSCheck|GetLOS_Check_DBNO_High_Data)))
  (Variables|LOSCheck|SetLOSCheck_DBNO_High_Cached (Math|Vector|MakeVector 0.0 0.0 (Math|Float|SelectFloat _outxy_14 0.0 _returnvalue_14)))
  (bind (_outxy_15 _returnvalue_15) (Fortnite|EvaluateCurveTableRow (Variables|LOSCheck|GetLOS_Check_DBNO_Low_Data)))
  (Variables|LOSCheck|SetLOSCheck_DBNO_Low_Cached (Math|Vector|MakeVector 0.0 0.0 (Math|Float|SelectFloat _outxy_15 -15.0 _returnvalue_15))))

(fn EvaluateCondition
  (params (ReloadStateData :type Segmented%20Reload%20State%20Data%20Structure%20%28by%20ref%29) (WeaponActor :type Fort%20Weapon%20Object%20Reference))
  (bind _asbp_fg_fang_pistol (Utilities|Casting|CastToBP_FG_Fang_Pistol WeaponActor)
    (:then
      (Variables|Default|SetFang_Pistol _asbp_fg_fang_pistol)
      (bind _returnvalue (Development|GetConsoleVariableBoolValue "FillerGrunt.MagThrowTargeting.EnableNative"))
      (if _returnvalue
        ; Nativized Version
        (bind _asfort_player_pawn (Utilities|Casting|CastToFortPlayerPawn (Actor|GetOwner (Variables|Default|GetFang_Pistol)))
          (:then
            (Utilities|IsValid (Class|FortAbilitySystemInterface|GetAbilitySystemComponentDontCreate _asfort_player_pawn)
              (:Is%20Valid
                (return (and (GameplayTags|HasMatchingGameplayTag (Class|FortAbilitySystemInterface|GetAbilitySystemComponentDontCreate _asfort_player_pawn) "(TagName=\"Ability.FillerGrunt.MagThrow.Targeting.InRange\")") (> (Utilities|Struct|BreakSegmentedReloadStateData ReloadStateData) 0))))
              (:Is%20Not%20Valid
                (return false))))
          (:CastFailed
            (return false)))
        (else
          ; Custom BP Trace (preferred)
          (CallFunction|CacheStatData)
          (bind _validinrange (CallFunction|RangeCheck))
          (Class|BPFGFangPistol|CacheMagThrowValidinRange (Variables|Default|GetFang_Pistol) _validinrange)
          (return (and _validinrange (> (Utilities|Struct|BreakSegmentedReloadStateData ReloadStateData) 0))))))
    (:CastFailed
      (return false))))

(fn Get_LOSCheck_HighLow
  (params (PlayerPawn :type Fort%20Player%20Pawn%20Object%20Reference))
  (LocalVariables|SetTarget PlayerPawn)
  (if (or (Vehicle|IsPassengerinVehicle (LocalVariables|GetTarget)) (Vehicle|IsDrivingVehicle (LocalVariables|GetTarget)))
    (return (+ (Transformation|GetActorLocation (LocalVariables|GetTarget)) (Variables|LOSCheck|GetLOSCheck_Vehicle_High_Cached)) (+ (Transformation|GetActorLocation (LocalVariables|GetTarget)) (Variables|LOSCheck|GetLOSCheck_Vehicle_Low_Cached)))
    (elif (Health|IsDBNO (LocalVariables|GetTarget))
      (return (+ (Transformation|GetActorLocation (LocalVariables|GetTarget)) (Variables|LOSCheck|GetLOSCheck_DBNO_High_Cached)) (+ (Transformation|GetActorLocation (LocalVariables|GetTarget)) (Variables|LOSCheck|GetLOSCheck_DBNO_Low_Cached)))
      (elif (Variables|/Script/Engine.Character|GetbIsCrouched (LocalVariables|GetTarget))
        (return (+ (Transformation|GetActorLocation (LocalVariables|GetTarget)) (Variables|LOSCheck|GetLOSCheck_Crouching_High_Cached)) (+ (Transformation|GetActorLocation (LocalVariables|GetTarget)) (Variables|LOSCheck|GetLOSCheck_Crouching_Low_Cached)))
        (else
          (return (+ (Transformation|GetActorLocation (LocalVariables|GetTarget)) (Variables|LOSCheck|GetLOSCheck_Standing_High_Cached)) (+ (Transformation|GetActorLocation (LocalVariables|GetTarget)) (Variables|LOSCheck|GetLOSCheck_Standing_Low_Cached))))))))

(fn IsWithinPyramidCone
  (params (Actor :type Actor%20Object%20Reference) (ConeTipLoc :type Vector) (ConeForward :type Vector) (ConeAngleHeight :type Float%20%28double-precision%29) (ConeAngleWidth :type Float%20%28double-precision%29))
  (bind _y (.y ConeForward))
  (bind _x (.x ConeForward))
  (bind _returnvalue (Transformation|GetActorLocation Actor))
  (bind _returnvalue_2 (Math|Rotator|FindLookatRotation ConeTipLoc _returnvalue))
  (bind (_returnvalue_x _returnvalue_y _returnvalue_z) (Math|Vector|BreakVector (Math|Vector|GetForwardVector _returnvalue_2)))
  (bind _returnvalue_3 (Math|Vector2D|DotProduct(2D) (Math|Vector2D|MakeVector2D _returnvalue_x _returnvalue_y) (Math|Vector2D|MakeVector2D _x _y)))
  (bind _returnvalue_4 (Math|Trig|Acos(Degrees) _returnvalue_3))
  (bind _z (.z ConeForward))
  (bind _returnvalue_5 (Math|Vector2D|DotProduct(2D) (Math|Vector2D|MakeVector2D _returnvalue_x _returnvalue_z) (Math|Vector2D|MakeVector2D _x _z)))
  (Development|PrintString Actor (Utilities|String|Append "Yaw: " (Utilities|String|ToString(Float) _returnvalue_4) " | Pitch: " (Utilities|String|ToString(Float) (Math|Trig|Acos(Degrees) _returnvalue_5))) false false "(R=0.000000,G=0.660000,B=1.000000,A=1.000000)" 5.0)
  (return (and (<= _returnvalue_4 ConeAngleWidth) (<= _returnvalue_5 ConeAngleHeight))))

(fn LoS_Check
  (params (StartLoc :type Vector) (EndLoc :type Vector) (Player :type Fort%20Player%20Pawn%20Object%20Reference) (Target :type Fort%20Player%20Pawn%20Object%20Reference))
  (LocalVariables|SetBlockingHit false)
  (if (or (Vehicle|IsPassengerinVehicle Target) (Vehicle|IsDrivingVehicle Target))
    ; But for now, if enemy is inside one, allow LoS to pass so player can attempt to throw at target. If it doesn't hit that's okay, it will most likely hit vehicle and we like that!
    (return true)
    (else
      (bind _outhits (Collision|MultiLineTraceByChannel Player StartLoc EndLoc "TraceTypeQuery1" true (Utilities|Array|MakeArray Player Target)))
      (bind _array_element (Utilities|Array|ForEachLoopwithBreak _outhits)
        (:LoopBody
          ; IGNORE ALL PLAYER PAWNS FOR BLOCKING LoS
          (Utilities|Casting|CastToFortPlayerPawn (.HitActor (Collision|BreakHitResult _array_element))
            (:CastFailed
              (if (Collision|BreakHitResult _array_element)
                (LocalVariables|SetBlockingHit true)))))
        (:Completed
          _))
      (return (not (LocalVariables|GetBlockingHit))))))

(fn RangeCheck
  (bind _playerpawn (LocalVariables|GetPlayerPawn))
  (bind _returnvalue (Pawn|GetController _playerpawn))
  (bind (_location _rotation) (Pawn|GetPlayerViewPoint _returnvalue))
  (bind _playerpawn_2 (LocalVariables|GetPlayerPawn))
  (bind _biscrouched (Variables|/Script/Engine.Character|GetbIsCrouched _playerpawn_2))
  (bind _standingoffset_cached (Variables|Spawn|GetStandingOffset_Cached))
  (bind _crouchedoffset_cached (Variables|Spawn|GetCrouchedOffset_Cached))
  (bind _returnvalue_2 (Math|Vector|SelectVector _crouchedoffset_cached _standingoffset_cached _biscrouched))
  (bind _returnvalue_3 (Vehicle|IsDrivingVehicle _playerpawn_2))
  (bind _returnvalue_4 (Vehicle|IsPassengerinVehicle _playerpawn_2))
  (bind _returnvalue_5 (or _returnvalue_4 _returnvalue_3))
  (bind _spawnvehicleoffset_cached (Variables|Spawn|GetSpawnVehicleOffset_Cached))
  (bind _returnvalue_6 (Math|Vector|SelectVector _spawnvehicleoffset_cached _returnvalue_2 _returnvalue_5))
  (bind _capsulecomponent (Variables|/Script/Engine.Character|GetCapsuleComponent _playerpawn_2))
  (bind _returnvalue_7 (Transformation|GetWorldLocation _capsulecomponent))
  (bind _returnvalue_8 (Transformation|GetActorRightVector _playerpawn_2))
  (bind _returnvalue_9 (Math|Vector|GetForwardVector _rotation))
  (bind _rightoffset_cached (Variables|Default|GetRightOffset_Cached))
  (bind _returnvalue_10 (* _returnvalue_8 _rightoffset_cached))
  (bind _returnvalue_11 (Utilities|Operators|Add _returnvalue_7 _returnvalue_6 _returnvalue_10))
  (bind _asfort_player_pawn (Utilities|Casting|CastToFortPlayerPawn (Actor|GetOwner (Variables|Default|GetFang_Pistol)))
    (:then
      (bind _output_get (LocalVariables|SetPlayerPawn _asfort_player_pawn))
      ; if not valid, it's a sim-proxy, cancel checks to save perf
      (Utilities|IsValid (Pawn|GetController _output_get)
        (:Is%20Valid
          (LocalVariables|SetSpawnLoc _returnvalue_11)
          (LocalVariables|SetSpawnForward _returnvalue_9)
          (bind _outhits (Collision|MultiBoxTraceForObjects (LocalVariables|GetPlayerPawn) _returnvalue_11 (+ _returnvalue_11 (* _returnvalue_9 (Variables|Default|GetTargetingRange_Cached))) (Math|Vector|MakeVector 0.0 (* (Variables|Default|GetTargetingRange_Cached) (Math|Trig|Tan(Degrees) (* (Variables|Default|GetTargetingConeYaw_Cached) 0.5))) (* (Variables|Default|GetTargetingRange_Cached) (Math|Trig|Tan(Degrees) (* (Variables|Default|GetTargetingConePitch_Cached) 0.5)))) _rotation (Utilities|Array|MakeArray "ObjectTypeQuery3")))
          (for _array_element _outhits
            ; Initial filter to only allow FortPlayerPawns
            (bind _asfort_player_pawn_2 (Utilities|Casting|CastToFortPlayerPawn (.HitActor (Collision|BreakHitResult _array_element))))
            (Utilities|Array|AddUnique (LocalVariables|GetPotentialTargets) _asfort_player_pawn_2))
          ; Filter out rest - Only hostile and other filtering settings
          (bind (_batleastoneactorleft _returnvalue_12) (Ability|FilterTargetActorList (LocalVariables|GetPlayerPawn) (LocalVariables|GetTargetFilter) (LocalVariables|GetPlayerPawn) (LocalVariables|GetPotentialTargets)))
          (if _batleastoneactorleft
            (LocalVariables|SetAtLeastOnePassedLOS false)
            (bind _array_element_2 (Utilities|Array|ForEachLoopwithBreak _returnvalue_12)
              (:LoopBody
                (bind _asfort_player_pawn_3 (Utilities|Casting|CastToFortPlayerPawn _array_element_2))
                (if (CallFunction|IsWithinPyramidCone _asfort_player_pawn_3 (LocalVariables|GetSpawnLoc) (LocalVariables|GetSpawnForward) (* (Variables|Default|GetTargetingConePitch_Cached) 0.5) (* (Variables|Default|GetTargetingConeYaw_Cached) 0.5))
                  (if (or (CallFunction|LoSCheck (LocalVariables|GetSpawnLoc) (CallFunction|GetLOSCheckHighLow _asfort_player_pawn_3) (LocalVariables|GetPlayerPawn) _asfort_player_pawn_3) (CallFunction|LoSCheck (LocalVariables|GetSpawnLoc) (.LowCheck (CallFunction|GetLOSCheckHighLow _asfort_player_pawn_3)) (LocalVariables|GetPlayerPawn) _asfort_player_pawn_3))
                    (LocalVariables|SetAtLeastOnePassedLOS true))))
              (:Completed
                _))
            (return (LocalVariables|GetAtLeastOnePassedLOS))
            (else
              (return false))))
        (:Is%20Not%20Valid
          (return true))))
    (:CastFailed
      (return false))))
