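; Gameplay ability for the Fang pistol's magazine throw, which launches a client-predicted projectile carrying a damage effect, with a separate Mythic variant.
; It caches curve-table tuning, then targets via box trace, hostile filter, angular cone and high/low line-of-sight checks.
; The valid target nearest the crosshair receives a lock-on homing projectile; otherwise a non-homing one flies along the view direction.
; Vehicle occupants automatically pass line-of-sight.
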(event Ability|EventActivateAbility
  (bind (_location _rotation) (Pawn|GetPlayerViewPoint (Pawn|GetController (Variables|Default|GetPlayerPawn))))
  (bind _currentclosesttarget (Variables|Default|GetCurrentClosestTarget))
  (bind _spawnlocation (Variables|Default|GetSpawnLocation))
  (bind (_event_data _fort_player_pawn) (Setup|SetupGenericAbility false false)
    (:Output
      (Utilities|IsValid _fort_player_pawn
        (:Is%20Valid
          (bind _output_get (Variables|Default|SetPlayerPawn _fort_player_pawn))
          (bind _asbp_fg_fang_pistol (Utilities|Casting|CastToBP_FG_Fang_Pistol (Variables|/Script/FortniteGame.FortPawn|GetCurrentWeapon _output_get)))
          (CallFunction|CacheStatData)
          ; NewCustom Targeting Version
          (bind _anyvalidtarget (Default|DoTargeting))
          (bind _returnvalue (Spec|MakeSpecHandle(ByClass) (select (Variables|Default|GetIsMythic) "/FillerGrunt/Gameplay/Pistol_Fang/Abilities/MagThrow/GE_MagThrow_Damage_Mythic.GE_MagThrow_Damage_Mythic_C" "/FillerGrunt/Gameplay/Pistol_Fang/Abilities/MagThrow/GE_MagThrow_Damage.GE_MagThrow_Damage_C") (Variables|Default|GetPlayerPawn) (Variables|Default|GetPlayerPawn)))
          (if _anyvalidtarget
            (bind _returnvalue_2 (Projectile|SpawnClientPredictedProjectile "/FillerGrunt/Gameplay/Pistol_Fang/Abilities/MagThrow/Prj_FangPistol_ThrownMag.Prj_FangPistol_ThrownMag_C" (Variables|Default|GetPlayerPawn) (Utilities|Struct|MakeSpawnProjectileParams _spawnlocation (Math|Conversions|RotationFromXVector (Math|Vector|RotateVectorAroundAxis (Math|Vector|GetForwardVector (Math|Rotator|FindLookatRotation (Variables|Default|GetSpawnLocation) (Transformation|GetActorLocation (Variables|Default|GetCurrentClosestTarget)))) (Variables|CachedStats|GetSpawnPitchOffset_Cached) (Math|Vector|GetRightVector (Math|Rotator|FindLookatRotation (Variables|Default|GetSpawnLocation) (Transformation|GetActorLocation (Variables|Default|GetCurrentClosestTarget)))))) (Math|Conversions|RotationFromXVector (Math|Vector|RotateVectorAroundAxis (Math|Vector|GetForwardVector (Math|Rotator|FindLookatRotation (Variables|Default|GetSpawnLocation) (Transformation|GetActorLocation (Variables|Default|GetCurrentClosestTarget)))) (Variables|CachedStats|GetSpawnPitchOffset_Cached) (Math|Vector|GetRightVector (Math|Rotator|FindLookatRotation (Variables|Default|GetSpawnLocation) (Transformation|GetActorLocation (Variables|Default|GetCurrentClosestTarget)))))) (Utilities|Struct|MakeFortGameplayEffectContainerSpec "(List=,bStopAtFirstSuccess=True,bKeepCheckingListOnIndestructibleHit=False,AbilityTargetBlockedGameplayTags=(GameplayTags=),bUseWeaponRanges=True,bUseMaxYawAngleToTarget=False,MaxYawAngleToTarget=0.000000)" (Utilities|Array|MakeArray _returnvalue)) "(TargetSelection=(List=,bStopAtFirstSuccess=True,bKeepCheckingListOnIndestructibleHit=False,AbilityTargetBlockedGameplayTags=(GameplayTags=),bUseWeaponRanges=True,bUseMaxYawAngleToTarget=False,MaxYawAngleToTarget=0.000000),TargetGameplayEffectSpecs=,OwnerGameplayEffectSpecs=,ActivationCues=(GameplayTags=),ImpactCues=(GameplayTags=),ImpactNoiseRange=0.000000,FlyByNoiseRange=0.000000,bOverrideChargeMagnitude=False,ChargeMagnitudeOverrideValue=0.000000)" false _currentclosesttarget (Variables|CachedStats|GetInitialSpeed_Cached) (Variables|CachedStats|GetMaxSpeed_Cached) (Variables|CachedStats|GetInitialGravity_Cached) 0.0 0 (Utilities|Struct|MakeFortProjectilePredictionKey)))) ; Updated to work with FortProjectileBase with client-side prediction.
            (Projectile|SetHomingLockTargetDistanceThreshold _returnvalue_2 _returnvalue_2 (* (Math|Vector|Distance(Vector) _spawnlocation (Transformation|GetActorLocation _currentclosesttarget)) 0.45) :!class /Script/FortniteGame.FortProjectileMovementInterface)
            (Projectile|SetHomingStyle _returnvalue_2 _returnvalue_2 "LockOn" false :!class /Script/FortniteGame.FortProjectileMovementInterface)
            (Projectile|SetHomingTargetOffset _returnvalue_2 _returnvalue_2 (Math|Vector|MakeVector 0.0 0.0 35.0) :!class /Script/FortniteGame.FortProjectileMovementInterface)
            (Projectile|SetHomingTarget _returnvalue_2 _returnvalue_2 _currentclosesttarget 0 :!class /Script/FortniteGame.FortProjectileMovementInterface)
            (Class|PrjFangPistolThrownMag|SetFollowTarget (Utilities|Casting|CastToPrj_FangPistol_ThrownMag _returnvalue_2) _currentclosesttarget)
            (Ability|EndAbility)
            (else
              (Projectile|SpawnClientPredictedProjectile "/FillerGrunt/Gameplay/Pistol_Fang/Abilities/MagThrow/Prj_FangPistol_ThrownMag_NoHoming.Prj_FangPistol_ThrownMag_NoHoming_C" (Variables|Default|GetPlayerPawn) (Utilities|Struct|MakeSpawnProjectileParams (Variables|Default|GetSpawnLocation) (Math|Conversions|RotationFromXVector (Math|Vector|RotateVectorAroundAxis (Math|Vector|GetForwardVector (Math|Rotator|FindLookatRotation (Variables|Default|GetSpawnLocation) (+ (Variables|Default|GetSpawnLocation) (* (Math|Vector|GetForwardVector _rotation) (Variables|Default|GetTargetingRange_Cached))))) (Variables|CachedStats|GetSpawnPitchOffset_Cached) (Math|Vector|GetRightVector (Math|Rotator|FindLookatRotation (Variables|Default|GetSpawnLocation) (+ (Variables|Default|GetSpawnLocation) (* (Math|Vector|GetForwardVector _rotation) (Variables|Default|GetTargetingRange_Cached))))))) (Math|Conversions|RotationFromXVector (Math|Vector|RotateVectorAroundAxis (Math|Vector|GetForwardVector (Math|Rotator|FindLookatRotation (Variables|Default|GetSpawnLocation) (+ (Variables|Default|GetSpawnLocation) (* (Math|Vector|GetForwardVector _rotation) (Variables|Default|GetTargetingRange_Cached))))) (Variables|CachedStats|GetSpawnPitchOffset_Cached) (Math|Vector|GetRightVector (Math|Rotator|FindLookatRotation (Variables|Default|GetSpawnLocation) (+ (Variables|Default|GetSpawnLocation) (* (Math|Vector|GetForwardVector _rotation) (Variables|Default|GetTargetingRange_Cached))))))) (Utilities|Struct|MakeFortGameplayEffectContainerSpec "(List=,bStopAtFirstSuccess=True,bKeepCheckingListOnIndestructibleHit=False,AbilityTargetBlockedGameplayTags=(GameplayTags=),bUseWeaponRanges=True,bUseMaxYawAngleToTarget=False,MaxYawAngleToTarget=0.000000)" (Utilities|Array|MakeArray _returnvalue)) "(TargetSelection=(List=,bStopAtFirstSuccess=True,bKeepCheckingListOnIndestructibleHit=False,AbilityTargetBlockedGameplayTags=(GameplayTags=),bUseWeaponRanges=True,bUseMaxYawAngleToTarget=False,MaxYawAngleToTarget=0.000000),TargetGameplayEffectSpecs=,OwnerGameplayEffectSpecs=,ActivationCues=(GameplayTags=),ImpactCues=(GameplayTags=),ImpactNoiseRange=0.000000,FlyByNoiseRange=0.000000,bOverrideChargeMagnitude=False,ChargeMagnitudeOverrideValue=0.000000)" false (Variables|Default|GetCurrentClosestTarget) (Variables|CachedStats|GetInitialSpeed_Cached) (Variables|CachedStats|GetMaxSpeed_Cached) (Variables|Default|GetNoHoming_Gravity_Cached) 0.0 0 (Utilities|Struct|MakeFortProjectilePredictionKey))) ; Updated to work with FortProjectileBase with client-side prediction.
              (Ability|EndAbility))))))))

(fn CacheStatData
  (Variables|Default|SetIsMythic (Utilities|Enum|Equal(Enum) (Rarity|GetWeaponDataRarity (Variables|/Script/FortniteGame.FortPawn|GetCurrentWeapon (Variables|Default|GetPlayerPawn))) "Mythic"))
  (bind (_outxy _returnvalue) (Fortnite|EvaluateCurveTableRow (Variables|Stats|GetInitialSpeedData)))
  (Variables|CachedStats|SetInitialSpeed_Cached (Math|Float|SelectFloat _outxy 3e+03 _returnvalue))
  (bind (_outxy_2 _returnvalue_2) (Fortnite|EvaluateCurveTableRow (Variables|Stats|GetMaxSpeedData)))
  (Variables|CachedStats|SetMaxSpeed_Cached (Math|Float|SelectFloat _outxy_2 3e+03 _returnvalue_2))
  (bind (_outxy_3 _returnvalue_3) (Fortnite|EvaluateCurveTableRow (Variables|Stats|GetInitialGravityData)))
  (Variables|CachedStats|SetInitialGravity_Cached (Math|Float|SelectFloat _outxy_3 0.0 _returnvalue_3))
  (bind (_outxy_4 _returnvalue_4) (Fortnite|EvaluateCurveTableRow (Variables|Stats|GetSpawnPitchOffsetData)))
  (Variables|CachedStats|SetSpawnPitchOffset_Cached (Math|Float|SelectFloat _outxy_4 -3.85 _returnvalue_4))
  (bind (_outxy_5 _returnvalue_5) (Fortnite|EvaluateCurveTableRow (Variables|Default|GetSpawnCrouchedOffsetData)))
  (Variables|CachedStats|SetCrouchedOffset_Cached (Math|Vector|MakeVector 0.0 0.0 (Math|Float|SelectFloat _outxy_5 35.0 _returnvalue_5)))
  (bind (_outxy_6 _returnvalue_6) (Fortnite|EvaluateCurveTableRow (Variables|Default|GetSpawnStandingOffsetData)))
  (Variables|CachedStats|SetStandingOffset_Cached (Math|Vector|MakeVector 0.0 0.0 (Math|Float|SelectFloat _outxy_6 7e+01 _returnvalue_6)))
  (bind (_outxy_7 _returnvalue_7) (Fortnite|EvaluateCurveTableRow (Variables|Default|GetSpawnVehicleOffsetData)))
  (Variables|Default|SetSpawnVehicleOffset_Cached (Math|Vector|MakeVector 0.0 0.0 (Math|Float|SelectFloat _outxy_7 35.0 _returnvalue_7)))
  (bind (_outxy_8 _returnvalue_8) (Fortnite|EvaluateCurveTableRow (Variables|Stats|GetRightOffsetData)))
  (Variables|CachedStats|SetRightOffset_Cached (Math|Float|SelectFloat _outxy_8 2e+01 _returnvalue_8))
  (bind (_outxy_9 _returnvalue_9) (Fortnite|EvaluateCurveTableRow (Variables|Default|GetTargetingRangeData)))
  (Variables|Default|SetTargetingRange_Cached (Math|Float|SelectFloat _outxy_9 2e+03 _returnvalue_9))
  (bind (_outxy_10 _returnvalue_10) (Fortnite|EvaluateCurveTableRow (Variables|Default|GetTargetingConePitchData)))
  (Variables|Default|SetTargetingConePitch_Cached (Math|Float|SelectFloat _outxy_10 45.0 _returnvalue_10))
  (bind (_outxy_11 _returnvalue_11) (Fortnite|EvaluateCurveTableRow (Variables|Default|GetTargetingConeYawData)))
  (Variables|Default|SetTargetingConeYaw_Cached (Math|Float|SelectFloat _outxy_11 8e+01 _returnvalue_11))
  (bind (_outxy_12 _returnvalue_12) (Fortnite|EvaluateCurveTableRow (Variables|Default|GetLOS_Check_Vehicle_High_Data)))
  (Variables|Default|SetLOSCheck_Vehicle_High_Cached (Math|Vector|MakeVector 0.0 0.0 (Math|Float|SelectFloat _outxy_12 4e+01 _returnvalue_12)))
  (bind (_outxy_13 _returnvalue_13) (Fortnite|EvaluateCurveTableRow (Variables|Default|GetLOS_Check_Vehicle_Low_Data)))
  (Variables|Default|SetLOSCheck_Vehicle_Low_Cached (Math|Vector|MakeVector 0.0 0.0 (Math|Float|SelectFloat _outxy_13 0.0 _returnvalue_13)))
  (bind (_outxy_14 _returnvalue_14) (Fortnite|EvaluateCurveTableRow (Variables|Default|GetLOS_Check_Standing_High_Data)))
  (Variables|Default|SetLOSCheck_Standing_High_Cached (Math|Vector|MakeVector 0.0 0.0 (Math|Float|SelectFloat _outxy_14 7e+01 _returnvalue_14)))
  (bind (_outxy_15 _returnvalue_15) (Fortnite|EvaluateCurveTableRow (Variables|Default|GetLOS_Check_Standing_Low_Data)))
  (Variables|Default|SetLOSCheck_Standing_Low_Cached (Math|Vector|MakeVector 0.0 0.0 (Math|Float|SelectFloat _outxy_15 35.0 _returnvalue_15)))
  (bind (_outxy_16 _returnvalue_16) (Fortnite|EvaluateCurveTableRow (Variables|Default|GetLOS_Check_Crouched_High_Data)))
  (Variables|Default|SetLOSCheck_Crouching_High_Cached (Math|Vector|MakeVector 0.0 0.0 (Math|Float|SelectFloat _outxy_16 35.0 _returnvalue_16)))
  (bind (_outxy_17 _returnvalue_17) (Fortnite|EvaluateCurveTableRow (Variables|Default|GetLOS_Check_Crouched_Low_Data)))
  (Variables|Default|SetLOSCheck_Crouching_Low_Cached (Math|Vector|MakeVector 0.0 0.0 (Math|Float|SelectFloat _outxy_17 17.5 _returnvalue_17)))
  (bind (_outxy_18 _returnvalue_18) (Fortnite|EvaluateCurveTableRow (Variables|Default|GetLOS_Check_DBNO_High_Data)))
  (Variables|Default|SetLOSCheck_DBNO_High_Cached (Math|Vector|MakeVector 0.0 0.0 (Math|Float|SelectFloat _outxy_18 0.0 _returnvalue_18)))
  (bind (_outxy_19 _returnvalue_19) (Fortnite|EvaluateCurveTableRow (Variables|Default|GetLOS_Check_DBNO_Low_Data)))
  (Variables|Default|SetLOSCheck_DBNO_Low_Cached (Math|Vector|MakeVector 0.0 0.0 (Math|Float|SelectFloat _outxy_19 -15.0 _returnvalue_19)))
  (Utilities|FlowControl|Sequence
    (:then_0
      (bind _outxy_20 (DataTable|EvaluateCurveTableRow "/FillerGrunt/DataTables/FillerGrunt_GameData.FillerGrunt_GameData" "FillerGrunt.FangPistol.MagThrow.Gravity.NoHoming.Initial")
        (:RowFound
          (Variables|Default|SetNoHoming_Gravity_Cached _outxy_20))
        (:RowNotFound
          (Variables|Default|SetNoHoming_Gravity_Cached 0.3))))
    (:then_1
      (Utilities|FlowControl|Sequence
        (:then_0
          (bind _outxy_21 (DataTable|EvaluateCurveTableRow "/FillerGrunt/DataTables/FillerGrunt_GameData.FillerGrunt_GameData" "FillerGrunt.FangPistol.MagThrow.Targeting.LockOn.Cone.PitchScale")
            (:RowFound
              (Variables|Default|SetConeLockOn_PitchScale _outxy_21))
            (:RowNotFound
              (Variables|Default|SetConeLockOn_PitchScale 0.3))))
        (:then_1
          (Utilities|FlowControl|Sequence
            (:then_0
              (bind _outxy_22 (DataTable|EvaluateCurveTableRow "/FillerGrunt/DataTables/FillerGrunt_GameData.FillerGrunt_GameData" "FillerGrunt.FangPistol.MagThrow.Targeting.LockOn.Cone.YawScale")
                (:RowFound
                  (Variables|Default|SetConeLockOn_YawScale _outxy_22))
                (:RowNotFound
                  (Variables|Default|SetConeLockOn_YawScale 0.3))))
            (:then_1
              (Utilities|FlowControl|Sequence
                (:then_0
                  (bind _outxy_23 (DataTable|EvaluateCurveTableRow "/FillerGrunt/DataTables/FillerGrunt_GameData.FillerGrunt_GameData" "FillerGrunt.FangPistol.MagThrow.Targeting.LockOn.Cone.PitchScale.Mythic")
                    (:RowFound
                      (Variables|Default|SetConeLockOn_PitchScale_Mythic _outxy_23))
                    (:RowNotFound
                      (Variables|Default|SetConeLockOn_PitchScale_Mythic 0.3))))
                (:then_1
                  (bind _outxy_24 (DataTable|EvaluateCurveTableRow "/FillerGrunt/DataTables/FillerGrunt_GameData.FillerGrunt_GameData" "FillerGrunt.FangPistol.MagThrow.Targeting.LockOn.Cone.PitchScale.Mythic")
                    (:RowFound
                      (Variables|Default|SetConeLockOn_YawScale_Mythic _outxy_24))
                    (:RowNotFound
                      (Variables|Default|SetConeLockOn_YawScale_Mythic 0.3))))))))))))

(fn DoTargeting
  (bind (_location _rotation) (Pawn|GetPlayerViewPoint (Pawn|GetController (Variables|Default|GetPlayerPawn))))
  (bind _playerpawn (Variables|Default|GetPlayerPawn))
  (bind _biscrouched (Variables|/Script/Engine.Character|GetbIsCrouched _playerpawn))
  (bind _standingoffset_cached (Variables|CachedStats|GetStandingOffset_Cached))
  (bind _crouchedoffset_cached (Variables|CachedStats|GetCrouchedOffset_Cached))
  (bind _returnvalue (Math|Vector|SelectVector _crouchedoffset_cached _standingoffset_cached _biscrouched))
  (bind _playerpawn_2 (Variables|Default|GetPlayerPawn))
  (bind _returnvalue_2 (Pawn|GetController _playerpawn_2))
  (bind (_location_2 _rotation_2) (Pawn|GetPlayerViewPoint _returnvalue_2))
  (bind _returnvalue_3 (Vehicle|IsDrivingVehicle _playerpawn))
  (bind _returnvalue_4 (Vehicle|IsPassengerinVehicle _playerpawn))
  (bind _returnvalue_5 (or _returnvalue_4 _returnvalue_3))
  (bind _spawnvehicleoffset_cached (Variables|Default|GetSpawnVehicleOffset_Cached))
  (bind _returnvalue_6 (Math|Vector|SelectVector _spawnvehicleoffset_cached _returnvalue _returnvalue_5))
  (bind _capsulecomponent (Variables|/Script/Engine.Character|GetCapsuleComponent _playerpawn))
  (bind _returnvalue_7 (Transformation|GetWorldLocation _capsulecomponent))
  (bind _returnvalue_8 (Transformation|GetActorRightVector _playerpawn))
  (bind _returnvalue_9 (Math|Vector|GetForwardVector _rotation_2))
  (bind _rightoffset_cached (Variables|CachedStats|GetRightOffset_Cached))
  (bind _returnvalue_10 (* _returnvalue_8 _rightoffset_cached))
  (bind _returnvalue_11 (Utilities|Operators|Add _returnvalue_7 _returnvalue_6 _returnvalue_10))
  (Variables|Default|SetSpawnLocation _returnvalue_11)
  (LocalVariables|SetSpawnForward _returnvalue_9)
  (bind _outhits (Collision|MultiBoxTraceForObjects _returnvalue_11 (+ _returnvalue_11 (* _returnvalue_9 (Variables|Default|GetTargetingRange_Cached))) (Math|Vector|MakeVector 0.0 (* (Variables|Default|GetTargetingRange_Cached) (Math|Trig|Tan(Degrees) (* (Variables|Default|GetTargetingConeYaw_Cached) (Math|Float|SelectFloat (Variables|Default|GetConeLockOn_YawScale_Mythic) (Variables|Default|GetConeLockOn_YawScale) (Variables|Default|GetIsMythic))))) (* (Variables|Default|GetTargetingRange_Cached) (Math|Trig|Tan(Degrees) (* (Variables|Default|GetTargetingConePitch_Cached) (Math|Float|SelectFloat (Variables|Default|GetConeLockOn_PitchScale_Mythic) (Variables|Default|GetConeLockOn_PitchScale) (Variables|Default|GetIsMythic)))))) _rotation_2 (Utilities|Array|MakeArray "ObjectTypeQuery3")))
  (for _array_element _outhits
    ; Initial filter to only allow FortPlayerPawns
    (bind _asfort_player_pawn (Utilities|Casting|CastToFortPlayerPawn (.HitActor (Collision|BreakHitResult _array_element))))
    (Utilities|Array|AddUnique (LocalVariables|GetPotentialTargets) _asfort_player_pawn))
  ; Filter out rest - Only hostile and other filtering settings
  (bind (_batleastoneactorleft _returnvalue_12) (Ability|FilterTargetActorList (LocalVariables|GetTargetFilter) (Variables|Default|GetPlayerPawn) (LocalVariables|GetPotentialTargets)))
  (if _batleastoneactorleft
    (Utilities|Array|Clear (Variables|Default|GetLOS_Passed_Targets))
    (bind _array_element_2 (Utilities|Array|ForEachLoopwithBreak _returnvalue_12)
      (:LoopBody
        (bind _asfort_player_pawn_2 (Utilities|Casting|CastToFortPlayerPawn _array_element_2))
        (if (Default|IsWithinPyramidCone _asfort_player_pawn_2 (Variables|Default|GetSpawnLocation) (LocalVariables|GetSpawnForward) (* (Variables|Default|GetTargetingConePitch_Cached) 0.5) (* (Variables|Default|GetTargetingConeYaw_Cached) 0.5))
          (if (or (Default|LoSCheck (Variables|Default|GetSpawnLocation) (Default|GetLOSCheckHighLow _asfort_player_pawn_2) (Variables|Default|GetPlayerPawn) _asfort_player_pawn_2) (Default|LoSCheck (Variables|Default|GetSpawnLocation) (.LowCheck (Default|GetLOSCheckHighLow _asfort_player_pawn_2)) (Variables|Default|GetPlayerPawn) _asfort_player_pawn_2))
            (Utilities|Array|AddUnique (Variables|Default|GetLOS_Passed_Targets) _asfort_player_pawn_2))))
      (:Completed
        (if (> (Utilities|Array|Length (Variables|Default|GetLOS_Passed_Targets)) 0)
          (Variables|Default|SetCurrentClosestTarget (Utilities|Array|Get(acopy) (Variables|Default|GetLOS_Passed_Targets)))
          (for _array_element_3 (Variables|Default|GetLOS_Passed_Targets)
            (CallFunction|TestClosesttoCrosshair _array_element_3 _location (Math|Vector|GetForwardVector _rotation)))
          (return true (Variables|Default|GetCurrentClosestTarget))
          (else
            (return false 0)))))
    (else
      (return false 0))))

(fn Get_LOSCheck_HighLow
  (params (PlayerPawn :type Fort%20Player%20Pawn%20Object%20Reference))
  (LocalVariables|SetTarget PlayerPawn)
  (if (or (Vehicle|IsPassengerinVehicle (LocalVariables|GetTarget)) (Vehicle|IsDrivingVehicle (LocalVariables|GetTarget)))
    (return (+ (Transformation|GetActorLocation (LocalVariables|GetTarget)) (Variables|Default|GetLOSCheck_Vehicle_High_Cached)) (+ (Transformation|GetActorLocation (LocalVariables|GetTarget)) (Variables|Default|GetLOSCheck_Vehicle_Low_Cached)))
    (elif (Health|IsDBNO (LocalVariables|GetTarget))
      (return (+ (Transformation|GetActorLocation (LocalVariables|GetTarget)) (Variables|Default|GetLOSCheck_DBNO_High_Cached)) (+ (Transformation|GetActorLocation (LocalVariables|GetTarget)) (Variables|Default|GetLOSCheck_DBNO_Low_Cached)))
      (elif (Variables|/Script/Engine.Character|GetbIsCrouched (LocalVariables|GetTarget))
        (return (+ (Transformation|GetActorLocation (LocalVariables|GetTarget)) (Variables|Default|GetLOSCheck_Crouching_High_Cached)) (+ (Transformation|GetActorLocation (LocalVariables|GetTarget)) (Variables|Default|GetLOSCheck_Crouching_Low_Cached)))
        (else
          (return (+ (Transformation|GetActorLocation (LocalVariables|GetTarget)) (Variables|Default|GetLOSCheck_Standing_High_Cached)) (+ (Transformation|GetActorLocation (LocalVariables|GetTarget)) (Variables|Default|GetLOSCheck_Standing_Low_Cached))))))))

(fn IsWithinPyramidCone
  (params (Actor :type Actor%20Object%20Reference) (ConeTipLoc :type Vector) (ConeForward :type Vector) (ConeAngleHeight :type Float%20%28double-precision%29) (ConeAngleWidth :type Float%20%28double-precision%29))
  (bind _y (.y ConeForward))
  (bind _x (.x ConeForward))
  (bind _returnvalue (Transformation|GetActorLocation Actor))
  (bind _returnvalue_2 (Math|Rotator|FindLookatRotation ConeTipLoc _returnvalue))
  (bind (_returnvalue_x _returnvalue_y _returnvalue_z) (Math|Vector|BreakVector (Math|Vector|GetForwardVector _returnvalue_2)))
  (bind _returnvalue_3 (Math|Vector2D|DotProduct(2D) (Math|Vector2D|MakeVector2D _returnvalue_x _returnvalue_y) (Math|Vector2D|MakeVector2D _x _y)))
  (bind _returnvalue_4 (Math|Trig|Acos(Degrees) _returnvalue_3))
  (bind _z (.z ConeForward))
  (bind _returnvalue_5 (Math|Vector2D|DotProduct(2D) (Math|Vector2D|MakeVector2D _returnvalue_x _returnvalue_z) (Math|Vector2D|MakeVector2D _x _z)))
  (Development|PrintString (Utilities|String|Append "Yaw: " (Utilities|String|ToString(Float) _returnvalue_4) " | Pitch: " (Utilities|String|ToString(Float) (Math|Trig|Acos(Degrees) _returnvalue_5))) false false "(R=0.000000,G=0.660000,B=1.000000,A=1.000000)" 5.0)
  (return (and (<= _returnvalue_4 ConeAngleWidth) (<= _returnvalue_5 ConeAngleHeight))))

(fn CanActivateAbility
  (params (ActorInfo :type Gameplay%20Ability%20Actor%20Info%20Structure) (Handle :type Gameplay%20Ability%20Spec%20Handle%20Structure))
  (bind _asfort_player_state (Utilities|Casting|CastToFortPlayerState (Utilities|Struct|BreakGameplayAbilityActorInfo ActorInfo))
    (:then
      (Utilities|IsValid (PlayerState|GetCurrentPawn _asfort_player_state)
        (:Is%20Valid
          (Utilities|Casting|CastToBP_FG_Fang_Pistol (Variables|/Script/FortniteGame.FortPawn|GetCurrentWeapon (PlayerState|GetCurrentPawn _asfort_player_state))
            (:then
              (return 0 true))
            (:CastFailed
              _)))
        (:Is%20Not%20Valid
          _)))
    (:CastFailed
      _))
  (return 0 false))

(fn LoS_Check
  (params (StartLoc :type Vector) (EndLoc :type Vector) (Player :type Fort%20Player%20Pawn%20Object%20Reference) (Target :type Fort%20Player%20Pawn%20Object%20Reference))
  (LocalVariables|SetBlockingHit false)
  (if (or (Vehicle|IsPassengerinVehicle Target) (Vehicle|IsDrivingVehicle Target))
    ; But for now, if enemy is inside one, allow LoS to pass so player can attempt to throw at target. If it doesn't hit that's okay, it will most likely hit vehicle and we like that!
    (return true)
    (else
      (bind _outhits (Collision|MultiLineTraceByChannel StartLoc EndLoc "TraceTypeQuery1" true (Utilities|Array|MakeArray Player Target)))
      (bind _array_element (Utilities|Array|ForEachLoopwithBreak _outhits)
        (:LoopBody
          ; IGNORE ALL PLAYER PAWNS FOR BLOCKING LoS
          (Utilities|Casting|CastToFortPlayerPawn (.HitActor (Collision|BreakHitResult _array_element))
            (:CastFailed
              (if (Collision|BreakHitResult _array_element)
                (LocalVariables|SetBlockingHit true)))))
        (:Completed
          _))
      (return (not (LocalVariables|GetBlockingHit))))))

(fn TestClosestToCrosshair
  (params (CheckTarget :type Fort%20Player%20Pawn%20Object%20Reference) (CameraLoc :type Vector) (CameraDir :type Vector))
  (if (not (== CheckTarget (Variables|Default|GetCurrentClosestTarget)))
    (if (> (Math|Vector|DotProduct CameraDir (Math|Vector|GetForwardVector (Math|Rotator|FindLookatRotation CameraLoc (Transformation|GetActorLocation CheckTarget)))) (Math|Vector|DotProduct CameraDir (Math|Vector|GetForwardVector (Math|Rotator|FindLookatRotation CameraLoc (Transformation|GetActorLocation (Variables|Default|GetCurrentClosestTarget))))))
      (Variables|Default|SetCurrentClosestTarget CheckTarget))))
