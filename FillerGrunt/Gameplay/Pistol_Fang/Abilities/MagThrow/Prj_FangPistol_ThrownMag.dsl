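; Homing thrown-magazine projectile for the Fang Pistol's MagThrow ability, spinning in flight.
; Curve table rows tune capsule size and vertical homing offsets, re-evaluated as the target pawn's stance, DBNO, vehicle or movement state changes.
; Reaching or passing the target resets homing turn speeds; spawning in water stops and destroys it.
; Impact and water-sink effects skip hidden server copies, so predicted client projectiles show them immediately.
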(event Collision|EventHit
  (params (MyComp :type Primitive%20Component%20Object%20Reference) (Other :type Actor%20Object%20Reference) (OtherComp :type Primitive%20Component%20Object%20Reference) (bSelfMoved :type Boolean) (HitLocation :type Vector :default "0, 0, 0") (HitNormal :type Vector :default "0, 0, 0") (NormalImpulse :type Vector :default "0, 0, 0") (Hit :type Hit%20Result%20Structure%20%28by%20ref%29))
  ;[1 0.9301 0.4072 0.502] With client prediction, the owning client gets two projectiles - the one it's predicting and the real-boi server version.
  ;
  ; Actor hidden in game is a way for us to tell which one is the client predicted and server one.
  ;
  ; If actor is visible, it is either an owning client or sim-proxy projectile.
  ; If NOT visible, it is the server versions.
  ; We only want to play cosmetic events, like hit sounds and impacts on the client so it doesn't need to wait for server to predict hits. So we ask for visibility to easily get the client version only.
  ;; Back up off hit slightly so FX doesn't clip through ground
  (if (Variables|Rendering|GetbHidden)
    _
    (elif (Variables|Default|GetHasTouchedWater)
      _
      (else
        (CallFunction|PlayHitSound (+ HitLocation (* HitNormal 2e+01)) HitNormal)))))

(event EventBeginPlay
  (bind _capsulecomponent (Variables|Collision|GetCapsuleComponent))
  ;[0.1154 0.6042 0 0.5 50] BEGIN PLAY
  (|CallParentFunction)
  (Utilities|FlowControl|Sequence
    (:then_0
      (bind _outxy (DataTable|EvaluateCurveTableRow "/FillerGrunt/DataTables/FillerGrunt_GameData.FillerGrunt_GameData" "FillerGrunt.FangPistol.MagThrow.Projectile.Hitbox.Radius")
        (:RowFound
          ;; Hotfix values that allow us to modify hitbox size
          ;; ORDER IS IMPORTANT. Setting half height first won't work
          (Components|Capsule|SetCapsuleRadius _capsulecomponent _outxy true)
          (Components|Capsule|SetCapsuleHalfHeight _capsulecomponent _outxy true))))
    (:then_1
      (if (Water|IsinWater (Variables|Water|GetWaterInteractionComponent))
        (Game|Components|ProjectileMovement|StopSimulating (Variables|Movement|GetProjectileMovementComponent) (Collision|MakeHitResult true false 0.0 0.0 (Transformation|GetWorldLocation (Variables|Collision|GetCapsuleComponent)) (Transformation|GetWorldLocation (Variables|Collision|GetCapsuleComponent)) (* (Transformation|GetForwardVector (Variables|Collision|GetCapsuleComponent)) -1.0) (* (Transformation|GetForwardVector (Variables|Collision|GetCapsuleComponent)) -1.0) 0 0 0 "None" "None" 0 0 0 (Transformation|GetWorldLocation (Variables|Collision|GetCapsuleComponent)) (Transformation|GetWorldLocation (Variables|Collision|GetCapsuleComponent))))
        (Utilities|FlowControl|DelayUntilNextFrame)
        (Actor|DestroyActor)))))

(event Custom|PostSpawn_SetHomingData
  (params (NewHomingData :type Projectile%20Homing%20Data%20Structure%20%28by%20ref%29) (IsLocationHoming :type Boolean))
  ; NEED TO FIX THIS.
  ; - To support location homing over target.
  ; - Use default homing data that the base class is using because that works way better than whatever settings I use here.
  ;
  (Projectile|SetHomingData 0 0 NewHomingData :!class /Script/FortniteGame.FortProjectileMovementInterface))

(event Custom|PlayHitSound
  (params (OffsetSpawn :type Vector) (HitNormal :type Vector))
  (bind _mesh_magazine (Variables|Default|GetMesh_Magazine))
  ;[0 0.197 1 0.5 50] HIT SOUND
  (Utilities|FlowControl|DoOnce false
    (:Completed
      (Rendering|SetVisibility _mesh_magazine false false)
      (Audio|PlaySoundatLocation (Variables|Default|GetSFX_Impact) (Transformation|GetWorldLocation _mesh_magazine) "0, 0, 0" 1.0 1.0 0.0 0 0 (Actor|GetOwner))
      (bind _returnvalue (Niagara|SpawnSystematLocation "/FillerGrunt/Assets/VFX/NS_FillerGrunt_Magazine_Eject.NS_FillerGrunt_Magazine_Eject" OffsetSpawn (Math|Rotator|MakeRotationfromAxes (Math|Vector|GetReflectionVector (Transformation|GetForwardVector (Variables|Collision|GetCapsuleComponent)) HitNormal) "0, 0, 0" (Math|Vector|VectorUp))))
      (Niagara|SetNiagaraVariable(Bool) _returnvalue "IsMagThrow" true))))

(event OnEnterWaterEvent(WaterInteractionComponent)
  (params (WaterBody :type Fort%20Water%20Body%20Actor%20Object%20Reference) (WaterInteractionComponent :type Fort%20Water%20Interaction%20Component%20Object%20Reference) (bIsFirstBody :type Boolean))
  (Variables|Default|SetHasTouchedWater true)
  ;[1 0.9301 0.4072 0.502] With client prediction, the owning client gets two projectiles - the one it's predicting and the real-boi server version.
  ;
  ; Actor hidden in game is a way for us to tell which one is the client predicted and server one.
  ;
  ; If actor is visible, it is either an owning client or sim-proxy projectile.
  ; If NOT visible, it is the server versions.
  ; We only want to play cosmetic events, like hit sounds and impacts on the client so it doesn't need to wait for server to predict hits. So we ask for visibility to easily get the client version only.
  (Networking|CanExecuteCosmeticEvents
    (:True
      (if (not (Variables|Rendering|GetbHidden))
        (CallFunction|PlayWaterSinkSound)))))

(event EventOnTouched
  (params (OtherActor :type Actor%20Object%20Reference) (OtherComp :type Primitive%20Component%20Object%20Reference) (HitResult :type Hit%20Result%20Structure%20%28by%20ref%29) (bIsOverlap :type Boolean))
  (bind (_bblockinghit _binitialoverlap _time _distance _location _impactpoint _normal _impactnormal _physmat _hitactor _hitcomponent _hitbonename _bonename _hititem _elementindex _faceindex _tracestart _traceend) (Collision|BreakHitResult HitResult))
  ;[1 0.9007 0 0.5 50] TOUCHED EVENT
  (|CallParentFunction OtherActor OtherComp HitResult bIsOverlap)
  ; Getting direction of last trace velocity when we touch, so that we can properly spawn FX... Rotation does follow velocity, but eh this worked.
  (Networking|CanExecuteCosmeticEvents
    (:True
      (Variables|Default|SetLastDir (Math|Vector|Normalize (- _traceend _tracestart))))))

(event Custom|Target_CrouchChanged
  (params (bCrouched :type Boolean) (Pawn :type Fort%20Pawn%20Object%20Reference))
  ;[0.15 0.15 0.15 0.5 50] TARGET - EVENT RESPONSE
  (CallFunction|MovementChanged))

(event Custom|Target_OnDBNOChanged
  (params (FortPawn :type Fort%20Pawn%20Object%20Reference) (bInIsDBNO :type Boolean))
  ;[0.15 0.15 0.15 0.5 50] TARGET - EVENT RESPONSE
  (CallFunction|MovementChanged))

(event Custom|Target_EnterWater
  (params (FortPlayerPawn :type Fort%20Player%20Pawn%20Object%20Reference))
  ;[0.15 0.15 0.15 0.5 50] TARGET - EVENT RESPONSE
  (CallFunction|EvaluateHomingTargetMovementStateOffset))

(event Custom|PlayWaterSinkSound
  (bind _waterinteractioncomponent (Variables|Water|GetWaterInteractionComponent))
  (bind (_waterplanelocation _waterplanenormal _watersurfaceposition _waterdepth _waterbodyidx) (FX|GetLastWaterSurfaceInfo _waterinteractioncomponent :!class /Script/FortniteGame.FortWaterInteractionComponent))
  ;[0 0.4162 1 0.5 50] WATER SINK EFFECTS
  (Audio|Components|Audio|Stop (Variables|Default|GetSFX_InAirLoop))
  (Utilities|FlowControl|DoOnce false
    (:Completed
      (Rendering|SetVisibility (Variables|Default|GetMesh_Magazine) false false)
      (bind _returnvalue (Niagara|SpawnSystematLocation "/FillerGrunt/Assets/VFX/NS_FillerGrunt_Magazine_Eject.NS_FillerGrunt_Magazine_Eject" (Transformation|GetWorldLocation (Variables|Default|GetPivotPoint_Magazine)) (Math|Conversions|RotationFromXVector (Math|Vector|Normalize (Variables|Default|GetLastDir)))))
      (Niagara|SetNiagaraVariable(Bool) _returnvalue "IsMagThrow" true)
      (Niagara|SetNiagaraVariable(Bool) _returnvalue "IsUnderwater" true)
      ;; While we do spawn a water hit impact in the GameplayCue, it looked great with this additive. Keeping
      (bind _returnvalue_2 (Effects|SpawnEmitteratLocation (Variables|Default|GetWater%20Impact%20FX) (Transformation|GetWorldLocation (Variables|Collision|GetCapsuleComponent)) (Math|Conversions|RotationFromXVector _waterplanenormal)))
      (Effects|Components|ParticleSystem|SetFloatParameter _returnvalue_2 "WaterDepth" _waterdepth)
      (Effects|Components|ParticleSystem|SetFloatParameter _returnvalue_2 "WaterBodyIndex" (Math|Conversions|ToFloat(Integer) _waterbodyidx)))))

(event Custom|Target_ExitWater
  (params (FortPlayerPawn :type Fort%20Player%20Pawn%20Object%20Reference) (bPawnIsOutsideOfAllWaterBodies :type Boolean))
  ;[0.15 0.15 0.15 0.5 50] TARGET - EVENT RESPONSE
  (CallFunction|EvaluateHomingTargetMovementStateOffset))

(event Collision|EventActorBeginOverlap
  (params (OtherActor :type Actor%20Object%20Reference)))

(event Custom|Target_ChangedMoveModes
  (params (Character :type Character%20Object%20Reference) (PrevMovementMode :type EMovementMode%20Enum :default "MOVE_None") (PreviousCustomMode :type Byte))
  ;[0.15 0.15 0.15 0.5 50] TARGET - EVENT RESPONSE
  (CallFunction|MovementChanged))

(event Custom|Target_OnVehicleChangeEvent
  (params (FortPlayerPawn :type Fort%20Player%20Pawn%20Object%20Reference) (NewVehicle :type Actor%20Object%20Reference) (OldVehicle :type Actor%20Object%20Reference))
  ;[0.15 0.15 0.15 0.5 50] TARGET - EVENT RESPONSE
  (CallFunction|MovementChanged))

(event EventTick
  (params (DeltaSeconds :type Float%20%28single-precision%29))
  ;[0.04655 0 1 0.5 50] TICK
  (Transformation|AddLocalRotation (Variables|Default|GetPivotPoint_Magazine) (Math|Rotator|MakeRotator 0.0 (* DeltaSeconds -4.5e+03)) false false))

(event Custom|MovementChanged
  ;[0.15 0.15 0.15 0.5 50] TARGET - EVENT RESPONSE
  ;; Movement change detected, change where the offset is!
  (bind _offset (CallFunction|EvaluateHomingTargetMovementStateOffset))
  (Projectile|SetHomingTargetOffset (Variables|Movement|GetProjectileMovementComponent) _offset false :!class /Script/FortniteGame.FortProjectileMovementComponent))

(event Custom|SetFollowTarget
  (params (FollowTarget :type Actor%20Object%20Reference))
  ;[0.15 0.15 0.15 0.5 50] FOLLOW TARGET - EVENT BINDS FOR MOVEMENT CHANGE
  (bind _asfort_player_pawn_athena (Utilities|Casting|CastToFortPlayerPawnAthena FollowTarget))
  (bind _output_get (Variables|Default|SetTargetPawn _asfort_player_pawn_athena))
  (CallFunction|MovementChanged)
  (Character|BindEventtoMovementModeChangedDelegate _output_get (EventDispatchers|CreateEvent :!event Custom|Target_ChangedMoveModes))
  (Pawn|BindEventtoOnPawnCrouchChanged _output_get (EventDispatchers|CreateEvent :!event Custom|Target_CrouchChanged))
  (Vehicle|BindEventtoOnVehicleStateChangedEvent _output_get (EventDispatchers|CreateEvent :!event Custom|Target_OnVehicleChangeEvent))
  (EventDispatchers|BindEventtoOnDBNOStateChanged _output_get (EventDispatchers|CreateEvent :!event Custom|Target_OnDBNOChanged))
  (EventDispatchers|BindEventtoOnPawnEnterWater _output_get (EventDispatchers|CreateEvent :!event Custom|Target_EnterWater))
  (EventDispatchers|BindEventtoOnPawnExitWater _output_get (EventDispatchers|CreateEvent :!event Custom|Target_ExitWater)))

(event ProjectileMovement|EventOnReachedLockTargetDistanceThreshold
  ;[0.15 0.15 0.15 0.5 50] TARGET DISTANCE THRESHOLD
  ; ON PASSED HOMING TARGET
  (Projectile|SetHomingTurnSpeedMin 0 0 0.0 :!class /Script/FortniteGame.FortProjectileMovementInterface)
  (Projectile|SetHomingTurnSpeedMax 0 0 1.5e+03 :!class /Script/FortniteGame.FortProjectileMovementInterface)
  (Projectile|SetHomingTimeUntilMaxTurnSpeed 0 0 0.2 :!class /Script/FortniteGame.FortProjectileMovementInterface))

(event ProjectileMovement|EventOnHomingPassedTarget
  ;[0.15 0.15 0.15 0.5 50] TARGET DISTANCE THRESHOLD
  ; ON PASSED HOMING TARGET
  (Projectile|SetHomingTurnSpeedMin 0 0 0.0 :!class /Script/FortniteGame.FortProjectileMovementInterface)
  (Projectile|SetHomingTurnSpeedMax 0 0 1.5e+03 :!class /Script/FortniteGame.FortProjectileMovementInterface)
  (Projectile|SetHomingTimeUntilMaxTurnSpeed 0 0 0.2 :!class /Script/FortniteGame.FortProjectileMovementInterface))

(fn EvaluateHomingTarget_MovementStateOffset
  (bind _targetpawn (Variables|Default|GetTargetPawn))
  (if (WaterBody|IsSurfaceSwimming _targetpawn)
    (if (Variables|/Script/FortniteGame.FortPawn|GetbIsDBNO _targetpawn)
      (return 0.0 0.0 (Variables|HomingOffsets|GetOffset_Swimming_DBNO))
      (else
        (return 0.0 0.0 (Variables|HomingOffsets|GetOffset_Swimming))))
    (elif (Variables|/Script/FortniteGame.FortPawn|GetbIsDBNO _targetpawn)
      (return 0.0 0.0 (Variables|HomingOffsets|GetOffset_DBNO))
      (elif (or (Vehicle|IsPassengerinVehicle _targetpawn) (Vehicle|IsDrivingVehicle _targetpawn))
        (return 0.0 0.0 (Variables|HomingOffsets|GetOffset_InVehicle))
        (elif (Variables|/Script/Engine.Character|GetbIsCrouched _targetpawn)
          (return 0.0 0.0 (Variables|HomingOffsets|GetOffset_Crouched))
          (else
            (return 0.0 0.0 (Variables|HomingOffsets|GetOffset_Standing))))))))

(fn ConstructionScript
  (bind _outxy (Fortnite|EvaluateCurveTableRow (Utilities|Struct|MakeCurveTableRowHandle "/FillerGrunt/DataTables/FillerGrunt_GameData.FillerGrunt_GameData" "FillerGrunt.FangPistol.MagThrow.Target.HomingOffset.Up.Crouching")))
  (Variables|HomingOffsets|SetOffset_Crouched _outxy)
  (bind _outxy_2 (Fortnite|EvaluateCurveTableRow (Utilities|Struct|MakeCurveTableRowHandle "/FillerGrunt/DataTables/FillerGrunt_GameData.FillerGrunt_GameData" "FillerGrunt.FangPistol.MagThrow.Target.HomingOffset.Up.Standing")))
  (Variables|HomingOffsets|SetOffset_Standing _outxy_2)
  (bind _outxy_3 (Fortnite|EvaluateCurveTableRow (Utilities|Struct|MakeCurveTableRowHandle "/FillerGrunt/DataTables/FillerGrunt_GameData.FillerGrunt_GameData" "FillerGrunt.FangPistol.MagThrow.Target.HomingOffset.Up.DBNO")))
  (Variables|HomingOffsets|SetOffset_DBNO _outxy_3)
  (bind _outxy_4 (Fortnite|EvaluateCurveTableRow (Utilities|Struct|MakeCurveTableRowHandle "/FillerGrunt/DataTables/FillerGrunt_GameData.FillerGrunt_GameData" "FillerGrunt.FangPistol.MagThrow.Target.HomingOffset.Up.Vehicle")))
  (Variables|HomingOffsets|SetOffset_InVehicle _outxy_4)
  (bind _outxy_5 (Fortnite|EvaluateCurveTableRow (Utilities|Struct|MakeCurveTableRowHandle "/FillerGrunt/DataTables/FillerGrunt_GameData.FillerGrunt_GameData" "FillerGrunt.FangPistol.MagThrow.Target.HomingOffset.Up.Swimming")))
  (Variables|HomingOffsets|SetOffset_Swimming _outxy_5)
  (bind _outxy_6 (Fortnite|EvaluateCurveTableRow (Utilities|Struct|MakeCurveTableRowHandle "/FillerGrunt/DataTables/FillerGrunt_GameData.FillerGrunt_GameData" "FillerGrunt.FangPistol.MagThrow.Target.HomingOffset.Up.Swimming.DBNO")))
  (Variables|HomingOffsets|SetOffset_Swimming_DBNO _outxy_6))
