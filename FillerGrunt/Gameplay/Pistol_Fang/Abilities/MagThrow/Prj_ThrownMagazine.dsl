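; Thrown-magazine projectile for the Fang pistol's MagThrow ability, which shows a gunshot-style sound indicator for hostile and neutral players on spawn.
; It explodes on hit or water entry, playing a splash in water and a server-side hit-marker sound on pawn hits.
; The server optionally spawns a pickup, offset along the impact normal outside water, and marks it through the instigator's marker component.
; Exploding disables collision and sets TurnOffFX, whose OnRep hides the mesh and trail.
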(event EventBeginPlay
  ;[1 1 1] Play Sound Indicator on begin play
  (|CallParentFunction)
  (Fortnite|ShowSoundIndicatorLocalforTeam :Location (Transformation|GetActorLocation) :MaxAudibleDistance (Variables|/Script/Engine.SoundBase|GetMaxDistance (Variables|Default|GetThrowIndicator_SoundRef)) :IndicatorType "Gunshot" :Team (Team|GetActorTeam self) :Affiliations (Utilities|Array|MakeArray "Hostile" "Neutral") :Tint "(R=0.000000,G=0.000000,B=0.000000,A=0.000000)" :AttenuationMaxRangeMultiplier 1.0 :AttenuationIntensityMultiplier 1.0)
  (if (Networking|IsServer)
    (Actor|SetOwner (Game|GetInstigator))))

(event Custom|Held%20Water%20Impact
  (Audio|PlaySoundatLocation (Variables|Default|GetSplash%20Sound_0) (Transformation|GetActorLocation)))

(event EventOnExploded
  (params (HitActors :type Array%20of%20Actor%20Object%20References) (HitResults :type Array%20of%20Hit%20Result%20Structures))
  (|CallParentFunction HitActors HitResults)
  (Variables|Default|SetTurnOffFX true)
  (Collision|SetActorEnableCollision false))

(event Custom|PlayHitSound
  ;[1 1 1] Play 2d hit sound on owner
  (Audio|PlaySoundatLocation (Variables|Default|GetHitMarkerSound) (Transformation|GetActorLocation)))

(event OnEnterWaterEvent(WaterInteractionComponent)
  (params (WaterBody :type Fort%20Water%20Body%20Actor%20Object%20Reference) (WaterInteractionComponent :type Fort%20Water%20Interaction%20Component%20Object%20Reference) (bIsFirstBody :type Boolean))
  (Utilities|FlowControl|Sequence
    (:then_0
      (if bIsFirstBody
        (CallFunction|HeldWaterImpact)))
    (:then_1
      ;[1 1 1] Explode on hit and spawn pickup
      (Delivery|DoExplosion)
      (if (Networking|IsServer)
        (Utilities|FlowControl|DoOnce false
          (:Completed
            (if (Variables|Default|GetShouldSpawnItem)
              (bind _returnvalue (Inventory|SpawnPickupInWorld (Variables|Default|GetItemToSpawn) 1 (select (Water|IsinWater (Variables|Water|GetWaterInteractionComponent)) (Transformation|GetActorLocation) (+ (Transformation|GetActorLocation) (* (.Normal (Collision|BreakHitResult (stale Collision|EventHit Hit))) (Variables|Default|GetImpactSpawnOffset)))) (Transformation|GetActorForwardVector) 0 true false false -1 "Other"))
              (CallFunction|MarkPickup _returnvalue))))))))

(event Collision|EventHit
  (params (MyComp :type Primitive%20Component%20Object%20Reference) (Other :type Actor%20Object%20Reference) (OtherComp :type Primitive%20Component%20Object%20Reference) (bSelfMoved :type Boolean) (HitLocation :type Vector :default "0, 0, 0") (HitNormal :type Vector :default "0, 0, 0") (NormalImpulse :type Vector :default "0, 0, 0") (Hit :type Hit%20Result%20Structure%20%28by%20ref%29))
  (bind (_bblockinghit _binitialoverlap _time _distance _location _impactpoint _normal _impactnormal _physmat _hitactor) (Collision|BreakHitResult Hit))
  ;[1 1 1] Explode on hit and spawn pickup
  (|CallParentFunction MyComp Other OtherComp bSelfMoved HitLocation HitNormal NormalImpulse Hit)
  (Utilities|FlowControl|DoOnce false
    (:Completed
      (Utilities|Casting|CastToFortPawn _hitactor
        (:then
          (if (Networking|IsServer)
            (CallFunction|PlayHitSound)))
        (:CastFailed
          _))
      (Development|PrintString (Utilities|GetDisplayName _hitactor))
      (Delivery|DoExplosion)
      (if (Networking|IsServer)
        (Utilities|FlowControl|DoOnce false
          (:Completed
            (if (Variables|Default|GetShouldSpawnItem)
              (bind _returnvalue (Inventory|SpawnPickupInWorld (Variables|Default|GetItemToSpawn) 1 (select (Water|IsinWater (Variables|Water|GetWaterInteractionComponent)) (Transformation|GetActorLocation) (+ (Transformation|GetActorLocation) (* _normal (Variables|Default|GetImpactSpawnOffset)))) (Transformation|GetActorForwardVector) 0 true false false -1 "Other"))
              (CallFunction|MarkPickup _returnvalue))))))))

(event Custom|MarkPickup
  (params (Pickup :type Fort%20Pickup%20Object%20Reference))
  (bind _asfort_player_pawn_athena (Utilities|Casting|CastToFortPlayerPawnAthena (Actor|GetOwner)))
  (bind _asfort_player_controller_athena (Utilities|Casting|CastToFortPlayerControllerAthena (Pawn|GetController _asfort_player_pawn_athena)))
  (MarkerComponent|MarkActoronServer (Class|FortPlayerController|GetMarkerComponent _asfort_player_controller_athena) Pickup true true))

(fn OnRep_TurnOffFX
  (bind _staticmesh (Variables|Default|GetStaticMesh))
  (bind _niagara_trail (Variables|Default|GetNiagara_Trail))
  (if (Variables|Default|GetTurnOffFX)
    (Components|Activation|Deactivate _niagara_trail)
    (Development|SetHiddeninGame _staticmesh true false)
    (else
      (Components|Activation|Activate _niagara_trail true)
      (Development|SetHiddeninGame _staticmesh false false))))

(fn ConstructionScript
  (|CallParentFunction))
